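# Physics asset for the QuietPeanuts guitar cosmetic, setting up two dangling ribbons for physical motion.
# Each ribbon is a three-capsule chain, widening toward the tip, that hangs from a shared root sphere with limited swing and twist at every joint.
# The root sphere is angularly locked to a large capsule on the handle bone, which anchors both chains to the instrument and gives it a collision shape.
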PhysicsAsset /SparksCosmetics/Instrument/Guitar/QuietPeanuts/Meshes/QuietPeanuts_Guitar_Physics
bodies {
 dyn_ribbon_a_01 [{shapeType=Capsule, length=2, radius=0.5, shapeName=dyn_ribbon_a_01_capsule, center=<1.5152 0 0>, rotation=<rot -90 90 -90>}]
 dyn_ribbon_a_02 [{shapeType=Capsule, length=2, radius=0.75, shapeName=dyn_ribbon_a_02_capsule, center=<2.1638 0 0>, rotation=<rot -90 179.2404 -179.2403>}]
 dyn_ribbon_a_03 [{shapeType=Capsule, length=2, radius=1, shapeName=dyn_ribbon_a_03_capsule, center=<1.9829 0 0>, rotation=<rot -90 6.7213 -6.7211>}]
 dyn_ribbon_b_01 [{shapeType=Capsule, length=2, radius=0.5, shapeName=dyn_ribbon_b_01_capsule, center=<1.5152 0 0>, rotation=<rot -90 180 179.9999>}]
 dyn_ribbon_b_02 [{shapeType=Capsule, length=2, radius=0.75, shapeName=dyn_ribbon_b_02_capsule, center=<2.1638 0 0>, rotation=<rot -90 160.5288 -160.5288>}]
 dyn_ribbon_b_03 [{shapeType=Capsule, length=2, radius=1, shapeName=dyn_ribbon_b_03_capsule, center=<1.9829 0 0>, rotation=<rot -90 -180 -180>}]
 dyn_ribbon_root [{shapeType=Sphere, radius=1, shapeName=dyn_ribbon_root_sphere}]
 handle [{shapeType=Capsule, length=15, radius=2.75, shapeName=handle_capsule, center=<0.3463 0.3385 69.4051>}]
}
constraints [
 {bone1Name=dyn_ribbon_a_01, bone2Name=dyn_ribbon_root, swing1LimitDegrees=45, swing2LimitDegrees=15, twistLimitDegrees=30}
 {bone1Name=dyn_ribbon_a_02, bone2Name=dyn_ribbon_a_01, swing1LimitDegrees=45, swing2LimitDegrees=15, twistLimitDegrees=30}
 {bone1Name=dyn_ribbon_a_03, bone2Name=dyn_ribbon_a_02, swing1LimitDegrees=45, swing2LimitDegrees=15, twistLimitDegrees=30}
 {bone1Name=dyn_ribbon_b_01, bone2Name=dyn_ribbon_root, swing1LimitDegrees=45, swing2LimitDegrees=15, twistLimitDegrees=30}
 {bone1Name=dyn_ribbon_b_02, bone2Name=dyn_ribbon_b_01, swing1LimitDegrees=45, swing2LimitDegrees=15, twistLimitDegrees=30}
 {bone1Name=dyn_ribbon_b_03, bone2Name=dyn_ribbon_b_02, swing1LimitDegrees=45, swing2LimitDegrees=15, twistLimitDegrees=30}
 {
  bone1Name=dyn_ribbon_root
  bone2Name=handle
  swing1LimitDegrees=45
  swing1Motion=Locked
  swing2LimitDegrees=45
  swing2Motion=Locked
  twistLimitDegrees=45
  twistMotion=Locked
 }
]
